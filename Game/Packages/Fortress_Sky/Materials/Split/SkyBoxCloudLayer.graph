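# Translucent unlit cloud layer material for the Fortress_Sky skybox.
# Twist-function UVs and time panning drive two samples of a normal-and-height map, whose height, thresholded from zenith to horizon and thinned near the sun, forms opacity.
# Derived normals, counter-rotated against the twist, fake lighting with sun tint, thickness-based rim light, ground color and zenith-to-horizon tint, blended toward fog.
# A static switch adds lightning masked by cloud density, Fresnel and elevation.
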Material /Game/Packages/Fortress_Sky/Materials/Split/SkyBoxCloudLayer
graph {
 properties {
  bComputeFogPerPixel=true
  bTangentSpaceNormal=false
  bUsedWithNanite=true
  blendMode=BLEND_Translucent
  numCustomizedUVs=6
  pixelDepthOffsetMode=PDOM_Legacy
  shadingModel=MSM_Unlit
  translucencyPass=MTP_BeforeDOF
  usedShadingModels=Unlit
 }
 nodes [
  ComponentMask(None=CollectionParameter() {collection=&FortniteMaterialParameters, parameterName=FogColor}) {b=true, g=true, r=true}
  Constant()
  Constant() {r=0.5}
  n1 = Subtract(A=Multiply(A=TextureCoordinate(), B=ScalarParameter() {defaultValue=16, parameterName=SkyTwistTile}))
  n2 = Constant() {r=0.01}
  n3 = Add(
   A=Multiply(A=@n1) {constB=0.5}
   B=MaterialFunctionCall(Speed=@n2) {materialFunction=&TimeWithSpeedVariable}.Time
  )
  n4 = ScalarParameter() {defaultValue=0.0525, parameterName=SkyTwistSpeed}
  n5 = MaterialFunctionCall!(
   SineAnimOffset
   Flow Direction=Constant2Vector() {g=0.5, r=-0.5}
   Time offset
   EdgeMaskSize
   UVS * size=@n1
   SineAnimSpeed
   Grid Noise Tile Rep
   Rotation amount modulator=Constant() {r=0.4}
   un frac Time=MaterialFunctionCall(Speed=@n4) {materialFunction=&TimeWithSpeedVariable}.Time
  ) {desc=15 fewer instructions on low end, materialFunction=/Game/Packages/Fortress_Sky/Fortnite_SkyTwistFunction}
  n6 = ComponentMask(None=@n5.Rotated UVs) {desc=rotated uvs, g=true, r=true}
  n7 = Multiply(A=Add(A=Multiply(A=@n3) {constB=0.125}, B=Multiply(A=@n6) {constB=0.125}))
  n8 = TextureSample(UVs=@n7) {
   samplerSource=SSM_Wrap_WorldGroupSettings
   samplerType=SAMPLERTYPE_LinearColor
   texture=/Game/Packages/Fortress_Sky/Materials/FinalCloudNormalAndHeightMap
  }
  n9 = OneMinus(None=@n8.B)
  n10 = Time()
  n11 = ScalarParameter() {defaultValue=-0.0193, parameterName=ZNormalATile}
  n12 = ScalarParameter() {defaultValue=24.563, parameterName=ZNormalTile}
  n13 = TextureSample(UVs=Add(A=Multiply(A=@n10, B=@n11), B=Multiply(A=@n7, B=@n12) {constA=3})) {
   samplerSource=SSM_Wrap_WorldGroupSettings
   samplerType=SAMPLERTYPE_LinearColor
   texture=&FinalCloudNormalAndHeightMap
  }
  n14 = Multiply(
   A=ScalarParameter() {defaultValue=0.1284, parameterName=ZNormalMultiply}
   B=Subtract(A=@n13.B, B=ScalarParameter() {defaultValue=-0.1772, parameterName=ZNormalMinus}) {constB=0.25}
  ) {constA=0.25, constB=0.1}
  n15 = CollectionParameter() {
   collection=&FortniteMaterialParameters
   parameterName=SunAndMoonModelWorldPosition
  }
  n16 = Normalize(VectorInput=Subtract(A=CameraPositionWS(), B=ComponentMask(None=@n15) {b=true, g=true, r=true}))
  n17 = DotProduct(A=CameraVectorWS(), B=@n16)
  n18 = ScalarParameter() {defaultValue=-0.75, parameterName=Clouds - Sun Tint Bias}
  n19 = Clamp(None=Multiply(A=Add(A=@n17, B=@n18), B=ScalarParameter() {defaultValue=4, parameterName=Clouds - Sun Tint Scale}))
  n20 = Multiply(A=@n19, B=@n19)
  n21 = Multiply(A=@n20, B=@n20)
  n22 = ScalarParameter() {defaultValue=0.2, parameterName=Cloud Dissipation Near Sun}
  n23 = ScalarParameter() {defaultValue=0.5, parameterName=Clouds - Zenith Opacity Threshold}
  n24 = ScalarParameter() {parameterName=Clouds - Horizon Opacity Threshold}
  n25 = TextureCoordinate() {coordinateIndex=1}
  n26 = Multiply(
   A=Subtract(A=ComponentMask(None=@n25) {g=true}, B=ScalarParameter() {parameterName=Clouds - Zenith to Horizon Bias})
   B=ScalarParameter() {defaultValue=1, parameterName=Clouds - Zenith to Horizon Scale}
  )
  n27 = Subtract(
   A=Add(A=@n8.B, B=Multiply(A=@n9, B=@n14))
   B=Add(A=Multiply(A=@n21, B=@n22) {constB=0.2, desc=amount of dissapation from sun vector}, B=LinearInterpolate(A=@n23, B=@n24, Alpha=@n26))
  )
  n28 = ComponentMask(None=TextureCoordinate() {coordinateIndex=1}) {g=true}
  n29 = Multiply(
   A=Subtract(A=@n28, B=ScalarParameter() {parameterName=Clouds - Zenith to Horizon Bias})
   B=ScalarParameter() {defaultValue=1, parameterName=Clouds - Zenith to Horizon Scale}
  )
  n30 = Multiply(
   A=@n27
   B=LinearInterpolate(
    A=ScalarParameter() {defaultValue=12.5, parameterName=Clouds - Zenith Opacity}
    B=ScalarParameter() {defaultValue=1.25, parameterName=Clouds - Horizon Opacity}
    Alpha=Clamp(None=@n29) {desc=h bright}
   ) {constA=12.5, constB=1.25, desc="a= 0"}
  ) {desc=opacity}
  DepthFade(Opacity=@n30, FadeDistance=Constant() {r=4500000}) {desc="+13 instructions > o"}
  n31 = ScalarParameter() {defaultValue=1, parameterName=DEBUG-StopTime}
  Multiply(A=@n31, B=@n4)
  Multiply(A=@n31, B=@n2)
  Multiply(A=@n31, B=@n11)
  Normalize(VectorInput=Subtract(A=CameraPositionWS(), B=WorldPosition().XYZ)) {desc=New Normal}
  Reroute(None=LinearInterpolate(B=@n30, Alpha=ScalarParameter() {defaultValue=0.5, parameterName=OpacityMaster}))
  n32 = VectorParameter() {parameterName=Clouds - Sun Tint Color, defaultValue=<rgba 1 1 1 1>}
  n33 = Clamp(None=@n27)
  n34 = Multiply(A=OneMinus(None=@n33), B=Constant() {r=0.05})
  n35 = OneMinus(None=@n33) {desc=block light when clouds are thick}
  n36 = Power(Base=@n35) {constExponent=10}
  n37 = ComponentMask(None=@n8.RGB) {g=true, r=true}
  n38 = ComponentMask(None=@n13.RGB) {g=true, r=true}
  n39 = Multiply(
   A=Power(Base=@n35, Exp=ScalarParameter() {defaultValue=23.3252, parameterName=detail normal power}) {constExponent=7}
   B=ScalarParameter() {defaultValue=0.3853, parameterName=Normal Lerp}
  )
  n40 = DeriveNormalZ(InXY=ConstantBiasScale(None=LinearInterpolate(A=@n37, B=@n38, Alpha=@n39) {constAlpha=0.05}) {bias=-0.5, scale=2})
  n41 = Constant3Vector() {constant=<rgba 0 0 1 0>}
  n42 = Constant()
  n43 = Multiply(A=@n5.RotationAngleValues) {constB=-1}
  n44 = Add(A=@n40, B=Multiply(A=RotateAboutAxis(NormalizedRotationAxis=@n41, RotationAngle=@n43, PivotPoint=@n42, Position=@n40) {desc=9}) {constB=0.5})
  n45 = VectorParameter() {parameterName=Normal Modulator, defaultValue=<rgba 1 1 1 1>}
  n46 = Transform(None=Normalize(VectorInput=Multiply(A=@n44, B=@n45.RGB)))
  n47 = Clamp(None=ComponentMask(None=@n46) {b=true}) {maxDefault=0, minDefault=-1}
  n48 = Multiply(A=@n47, B=@n47)
  n49 = Add(
   A=Multiply(A=Multiply(A=@n21, B=@n32.RGB), B=Add(A=@n34, B=@n36) {desc=block light - cloud thickness})
   B=Multiply(A=Multiply(A=@n33, B=@n48), B=VectorParameter() {parameterName=Ground Cloud Color, defaultValue=<rgba 0.28 0.266 0.0641 1>}.RGB)
  )
  n50 = VectorParameter() {parameterName=Clouds - Direct Sun Light Color, defaultValue=<rgba 0.5 0.25 0.25 1>}
  n51 = Multiply(A=@n50.RGB, B=@n36)
  n52 = Power(Base=@n35, Exp=ScalarParameter() {defaultValue=10, parameterName=Cloud Transparency Edge Lighting Falloff}) {constExponent=20}
  n53 = DotProduct(A=@n16, B=@n46)
  n54 = Clamp(None=ConstantBiasScale(None=Multiply(A=@n53) {constB=-1}) {bias=0.75, scale=0.5714})
  n55 = Add(A=Multiply(A=Multiply(A=@n52) {constB=2}, B=@n54), B=@n54)
  n56 = VectorParameter() {parameterName=Clouds - Zenith Color, defaultValue=<rgba 1 1 1 1>}
  n57 = VectorParameter() {parameterName=Clouds - Horizon Color, defaultValue=<rgba 1 1 1 1>}
  n58 = ScalarParameter() {parameterName=Clouds - Color - Zenith to Horizon Bias}
  n59 = Clamp(None=Multiply(A=Subtract(A=@n28, B=@n58), B=ScalarParameter() {defaultValue=1, parameterName=Clouds - Color - Zenith to Horizon Scale})) {desc=h bright}
  n60 = Add(
   A=Add(A=@n49, B=Multiply(A=@n51) {constB=0.1, desc=adds sunlight to the brim of the clouds})
   B=Multiply(A=Multiply(A=@n55, B=@n50.RGB), B=LinearInterpolate(A=@n56.RGB, B=@n57.RGB, Alpha=@n59))
  )
  n61 = VectorParameter() {parameterName=FogColor, defaultValue=<rgba 0 0 0 1>}
  n62 = ComponentMask(None=TextureCoordinate() {coordinateIndex=1}) {g=true}
  n63 = OneMinus(None=Clamp(None=OneMinus(None=@n62)))
  Reroute(None=Clamp(None=LinearInterpolate(A=@n60, B=@n61.RGB, Alpha=@n63)) {maxDefault=1000})
  n64 = VectorParameter() {desc="a = animated intensity", group=Lightning, parameterName=LightningColor, defaultValue=<rgba 0 0 0 1>}
  n65 = ComponentMask(None=@n46) {g=true, r=true}
  n66 = Multiply(
   A=OneMinus(None=@n64.A)
   B=MaterialFunctionCall(f2=Abs(None=@n65)) {materialFunction=&AddComponents}.f2
  ) {constB=0.2, desc=inverse spread}
  n67 = MaterialFunctionCall(Normal Vector=@n46, Invert Fresnel=StaticBool() {value=true}, Power=Constant() {r=4}) {materialFunction=&Fresnel_Function}
  n68 = Multiply(A=@n27)
  n69 = Add(A=Multiply(A=OneMinus(None=@n66), B=@n67.Result), B=OneMinus(None=@n68))
  n70 = Multiply(
   A=Subtract(A=@n68, B=ScalarParameter() {defaultValue=0.2, group=Lightning, parameterName=LightingThresholdOffset})
   B=ScalarParameter() {defaultValue=1.5, group=Lightning, parameterName="Lightning mask contrast value "}
  ) {constB=2}
  n71 = Add(A=Multiply(A=Multiply(A=@n64.A, B=@n69), B=@n35), B=OneMinus(None=Clamp(None=@n70)))
  n72 = Multiply(A=@n64.RGB, B=@n64.RGB)
  n73 = ComponentMask(None=TextureCoordinate() {coordinateIndex=1}) {g=true}
  n74 = Multiply(A=Multiply(A=Multiply(A=@n71, B=@n72), B=@n67.Result) {desc=only get facing edges}, B=Multiply(A=@n73, B=OneMinus(None=@n73)) {constB=2.5})
  StaticSwitch(True=Add(A=Multiply(A=@n74, B=@n63) {desc=remove lightning from the horizon}, B=@n49) {constB=0}, False=@n49) {defaultValue=true}
  TextureCoordinate() {uTiling=16, vTiling=16}
 ]
}
